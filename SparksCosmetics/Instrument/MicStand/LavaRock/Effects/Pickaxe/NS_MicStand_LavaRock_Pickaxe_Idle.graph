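# Idle Niagara effect for the LavaRock MicStand pickaxe cosmetic, emitting glowing ash and smoke from its skeletal mesh.
# Three GPU ash emitters spawn at offset bones or sockets, drift upward with curl noise, and cool from HDR orange to black.
# A CPU sprite emitter adds SubUV-animated grey smoke at two points, fading out under drag and noise.
# Ash emitters compute an owner-speed spawn probability but leave it switched off.
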NiagaraSystem /SparksCosmetics/Instrument/MicStand/LavaRock/Effects/Pickaxe/NS_MicStand_LavaRock_Pickaxe_Idle
dependencies {
 usedDataInterfaces [
  &NiagaraDataInterfaceArrayInt32
  &NiagaraDataInterfaceArrayFloat3
  &NiagaraDataInterfaceCurve
  &NiagaraDataInterfaceVector2DCurve
  &NiagaraDataInterfaceSpriteRendererInfo
  &NiagaraDataInterfaceColorCurve
  &NiagaraDataInterfaceSkeletalMesh
 ]
 usedDynamicInputs [&Vector2DFromCurve, &RandomRangeVector2D]
 usedModules [
  &self:SystemSpawnScript.SystemScriptSource.SystemScriptGraph.NiagaraNodeAssignment_0.SetVariables_#1
  &SystemState
 ]
}
emitters {
 Ashes {
  simTarget=GPUComputeSim
  emitterUpdateScript [
   &EmitterState(
    Life Cycle Mode=::NewEnumerator1
    Loop Duration=&RandomRangeFloat(Minimum=0.65, Maximum=4)
    Recalculate Duration Each Loop=true
    UseLoopDelay=false
   )
   &SpawnRate(
    Spawn Probability=&MakeCustomFloatFromBool(
     Bool=&SetBoolByFloatComparison(A=&VectorLength(Vector=$Engine.Owner.Velocity), B=200, Comparison Type=::NewEnumerator6)
     True Float=1
     False Float=0
    )
    SpawnRate=&RandomRangeFloat(Minimum=5, Maximum=7)
    Use Spawn Probability=false
   )
  ]
  particleSpawnScript [
   &InitializeParticle(
    Color=<rgba 20 20 20 1>
    Lifetime Mode=::NewEnumerator1
    Sprite Rotation Mode=::NewEnumerator2
    Sprite Size Mode=::NewEnumerator1
    Uniform Sprite Size Max=3.25
    Uniform Sprite Size Min=1.75
    UsePositionOffset=false
    Write Material Random=true
   )
   &PartitionParticles(Number Of Partitions=3)
   &SkeletalMeshLocation(
    Bone / Socket Index=&SelectIntFromArray(Direct Array Index=$Particles.PartitionParticles.Partition, Int Selection Array=DataInterface() {propertyValues {IntData [0,1,2]}})
    Bone / Socket Sampling Mode=::NewEnumerator5
    Interpolate Mesh Sampling=false
    MaskSpawnGroup=false
    Mesh Sampling Type=::NewEnumerator6
    Position Sampling=::NewEnumerator1
    Sampled Position Offset=&SelectVectorFromArray(
     Direct Array Index=$Particles.PartitionParticles.Partition
     Vector Selection Array=DataInterface() {propertyValues {FloatData [<-5 0 0>, <0 0 0>, <20 0 0>]}}
    )
    Skeletal Mesh=$System.Skeletal Mesh
    Source Movement Scale:Vector3f=<0.2 0.2 0.2>
   )
   &ShapeLocation(
    Cylinder Height=12
    Cylinder Radius=4
    MaskSpawnGroup=false
    Override Seed=false
    Shape Primitive=::NewEnumerator1
    Surface Only=false
    UseOwnerScale=true
   )
  ]
  particleUpdateScript [
   &ParticleState()
   &ScaleSpriteSize(
    Scale Sprite Size Mode=::NewEnumerator0
    Uniform Scale Factor=&MakeFloatFromVector(Vector=$Engine.Owner.Scale, Channel=::NewEnumerator2)
   ) {enabled=false, module=ScaleSpriteSize001}
   &AccelerationForce(Acceleration=<0 0 5>)
   &CurlNoiseForce(Noise Frequency=100, Noise Strength=5, Pan Noise=false)
   &SolveForcesAndVelocity(Clamp Velocity=false, Limit Acceleration=false)
   &ScaleSpriteSize(Uniform Curve Sprite Scale=DataInterface() {propertyValues {Curve {keys [
        {arriveTangent=5.5195, interpMode=RCIM_Cubic, leaveTangent=5.5195, tangentMode=RCTM_User, time=0, value=0}
        {arriveTangentWeight=0.25, leaveTangentWeight=0.0932, time=0.75, value=1}
        {arriveTangentWeight=0.4331, time=1, value=0}
       ]}}})
   &ScaleColor(
    Scale RGB=&MakeVectorFromLinearColorRGB(Color=&ColorFromCurve(ColorCurve=DataInterface() {propertyValues {
        AlphaCurve {keys []}
        BlueCurve {keys [
          {interpMode=RCIM_Cubic, leaveTangentWeight=0.2267, tangentMode=RCTM_User, time=0, value=49.0353}
          {arriveTangentWeight=11.117, interpMode=RCIM_Cubic, leaveTangentWeight=5.2315, tangentMode=RCTM_User, time=0.68, value=0}
          {arriveTangentWeight=0.1067, interpMode=RCIM_Cubic, tangentMode=RCTM_User, time=1, value=0}
         ]}
        GreenCurve {keys [
          {interpMode=RCIM_Cubic, leaveTangentWeight=0.2267, tangentMode=RCTM_User, time=0, value=154.5329}
          {arriveTangentWeight=35.0282, interpMode=RCIM_Cubic, leaveTangentWeight=16.4839, tangentMode=RCTM_User, time=0.68, value=0.426}
          {arriveTangentWeight=0.1067, interpMode=RCIM_Cubic, tangentMode=RCTM_User, time=1, value=0}
         ]}
        RedCurve {keys [
          {interpMode=RCIM_Cubic, leaveTangentWeight=0.2267, tangentMode=RCTM_User, time=0, value=500}
          {arriveTangentWeight=113.3336, interpMode=RCIM_Cubic, leaveTangentWeight=53.3334, tangentMode=RCTM_User, time=0.68, value=10}
          {arriveTangentWeight=0.1067, interpMode=RCIM_Cubic, tangentMode=RCTM_User, time=1, value=0}
         ]}
       }}))
    ScaleA=true
    ScaleRGB=true
   )
   &SpriteRotationRate(Rotation Rate=&RandomRangeFloat(Minimum=-3, Maximum=3))
  ]
  rendererClasses [&NiagaraSpriteRendererProperties]
  renderers [{rendererClass=&NiagaraSpriteRendererProperties}]
 }
 Ashes001 {
  simTarget=GPUComputeSim
  emitterUpdateScript [
   &EmitterState(
    Life Cycle Mode=::NewEnumerator1
    Loop Duration=&RandomRangeFloat(Minimum=0.65, Maximum=4)
    Recalculate Duration Each Loop=true
    UseLoopDelay=false
   )
   &SpawnRate(
    Spawn Probability=&MakeCustomFloatFromBool(
     Bool=&SetBoolByFloatComparison(A=&VectorLength(Vector=$Engine.Owner.Velocity), B=200, Comparison Type=::NewEnumerator6)
     True Float=1
     False Float=0
    )
    SpawnRate=&RandomRangeFloat(Minimum=1, Maximum=2)
    Use Spawn Probability=false
   )
  ]
  particleSpawnScript [
   &InitializeParticle(
    Color=<rgba 20 20 20 1>
    Lifetime Max=1.5
    Lifetime Mode=::NewEnumerator1
    Position Offset=<0 0 20>
    Sprite Rotation Mode=::NewEnumerator2
    Sprite Size Mode=::NewEnumerator1
    Uniform Sprite Size Max=3.25
    Uniform Sprite Size Min=1.75
    UsePositionOffset=false
    Write Material Random=true
   )
   &SkeletalMeshLocation(
    Bone / Socket Sampling Mode=::NewEnumerator5
    Interpolate Mesh Sampling=false
    MaskSpawnGroup=false
    Mesh Sampling Type=::NewEnumerator6
    Position Sampling=::NewEnumerator1
    Sampled Position Offset=<7 -14.5 0>
    Skeletal Mesh=$System.Skeletal Mesh
    Source Movement Scale:Vector3f=<0.2 0.2 0.2>
   )
   &ShapeLocation(Hemisphere Distribution=<1 0.5>, MaskSpawnGroup=false, Override Seed=false, Sphere Radius=2, UseOwnerScale=true)
  ]
  particleUpdateScript [
   &ParticleState()
   &ScaleSpriteSize(
    Scale Sprite Size Mode=::NewEnumerator0
    Uniform Scale Factor=&MakeFloatFromVector(Vector=$Engine.Owner.Scale, Channel=::NewEnumerator2)
   ) {enabled=false, module=ScaleSpriteSize001}
   &AccelerationForce(Acceleration=<0 0 5>)
   &CurlNoiseForce(Noise Frequency=100, Noise Strength=5, Pan Noise=false)
   &SolveForcesAndVelocity(Clamp Velocity=false, Limit Acceleration=false)
   &ScaleSpriteSize(Uniform Curve Sprite Scale=DataInterface() {propertyValues {Curve {keys [
        {arriveTangent=5.5195, interpMode=RCIM_Cubic, leaveTangent=5.5195, tangentMode=RCTM_User, time=0, value=0}
        {arriveTangentWeight=0.25, leaveTangentWeight=0.0932, time=0.75, value=1}
        {arriveTangentWeight=0.4331, time=1, value=0}
       ]}}})
   &ScaleColor(
    Scale RGB=&MakeVectorFromLinearColorRGB(Color=&ColorFromCurve(ColorCurve=DataInterface() {propertyValues {
        AlphaCurve {keys []}
        BlueCurve {keys [
          {interpMode=RCIM_Cubic, leaveTangentWeight=0.2267, tangentMode=RCTM_User, time=0, value=49.0353}
          {arriveTangentWeight=11.117, interpMode=RCIM_Cubic, leaveTangentWeight=5.2315, tangentMode=RCTM_User, time=0.68, value=0}
          {arriveTangentWeight=0.1067, interpMode=RCIM_Cubic, tangentMode=RCTM_User, time=1, value=0}
         ]}
        GreenCurve {keys [
          {interpMode=RCIM_Cubic, leaveTangentWeight=0.2267, tangentMode=RCTM_User, time=0, value=154.5329}
          {arriveTangentWeight=35.0282, interpMode=RCIM_Cubic, leaveTangentWeight=16.4839, tangentMode=RCTM_User, time=0.68, value=0.426}
          {arriveTangentWeight=0.1067, interpMode=RCIM_Cubic, tangentMode=RCTM_User, time=1, value=0}
         ]}
        RedCurve {keys [
          {interpMode=RCIM_Cubic, leaveTangentWeight=0.2267, tangentMode=RCTM_User, time=0, value=500}
          {arriveTangentWeight=113.3336, interpMode=RCIM_Cubic, leaveTangentWeight=53.3334, tangentMode=RCTM_User, time=0.68, value=10}
          {arriveTangentWeight=0.1067, interpMode=RCIM_Cubic, tangentMode=RCTM_User, time=1, value=0}
         ]}
       }}))
    ScaleA=true
    ScaleRGB=true
   )
   &SpriteRotationRate(Rotation Rate=&RandomRangeFloat(Minimum=-3, Maximum=3))
  ]
  rendererClasses [&NiagaraSpriteRendererProperties]
  renderers [{rendererClass=&NiagaraSpriteRendererProperties}]
 }
 Ashes_Base {
  simTarget=GPUComputeSim
  emitterUpdateScript [
   &EmitterState(
    Life Cycle Mode=::NewEnumerator1
    Loop Duration=&RandomRangeFloat(Minimum=0.65, Maximum=4)
    Recalculate Duration Each Loop=true
    UseLoopDelay=false
   )
   &SpawnRate(
    Spawn Probability=&MakeCustomFloatFromBool(
     Bool=&SetBoolByFloatComparison(A=&VectorLength(Vector=$Engine.Owner.Velocity), B=200, Comparison Type=::NewEnumerator6)
     True Float=1
     False Float=0
    )
    SpawnRate=&RandomRangeFloat(Minimum=7, Maximum=15)
    Use Spawn Probability=false
   )
  ]
  particleSpawnScript [
   &InitializeParticle(
    Color=<rgba 20 20 20 1>
    Lifetime Mode=::NewEnumerator1
    Sprite Rotation Mode=::NewEnumerator2
    Sprite Size Mode=::NewEnumerator1
    Uniform Sprite Size Max=3.25
    Uniform Sprite Size Min=1.75
    UsePositionOffset=false
    Write Material Random=true
   )
   &SkeletalMeshLocation(
    Bone / Socket Index=3
    Bone / Socket Sampling Mode=::NewEnumerator5
    Interpolate Mesh Sampling=false
    MaskSpawnGroup=false
    Mesh Sampling Type=::NewEnumerator6
    Position Sampling=::NewEnumerator1
    Sampled Position Offset=<7 0 0>
    Skeletal Mesh=$System.Skeletal Mesh
    Source Movement Scale:Vector3f=<0.2 0.2 0.2>
   )
   &ShapeLocation(
    Cylinder Height=6
    Cylinder Radius=13
    MaskSpawnGroup=false
    Override Seed=false
    Shape Primitive=::NewEnumerator1
    Surface Only=false
    UseOwnerScale=true
   )
  ]
  particleUpdateScript [
   &ParticleState()
   &ScaleSpriteSize(
    Scale Sprite Size Mode=::NewEnumerator0
    Uniform Scale Factor=&MakeFloatFromVector(Vector=$Engine.Owner.Scale, Channel=::NewEnumerator2)
   ) {enabled=false, module=ScaleSpriteSize001}
   &AccelerationForce(Acceleration=<0 0 5>)
   &CurlNoiseForce(Noise Frequency=100, Noise Strength=5, Pan Noise=false)
   &SolveForcesAndVelocity(Clamp Velocity=false, Limit Acceleration=false)
   &ScaleSpriteSize(Uniform Curve Sprite Scale=DataInterface() {propertyValues {Curve {keys [
        {arriveTangent=5.5195, interpMode=RCIM_Cubic, leaveTangent=5.5195, tangentMode=RCTM_User, time=0, value=0}
        {arriveTangentWeight=0.25, leaveTangentWeight=0.0932, time=0.75, value=1}
        {arriveTangentWeight=0.4331, time=1, value=0}
       ]}}})
   &ScaleColor(
    Scale RGB=&MakeVectorFromLinearColorRGB(Color=&ColorFromCurve(ColorCurve=DataInterface() {propertyValues {
        AlphaCurve {keys []}
        BlueCurve {keys [
          {interpMode=RCIM_Cubic, leaveTangentWeight=0.2267, tangentMode=RCTM_User, time=0, value=49.0353}
          {arriveTangentWeight=11.117, interpMode=RCIM_Cubic, leaveTangentWeight=5.2315, tangentMode=RCTM_User, time=0.68, value=0}
          {arriveTangentWeight=0.1067, interpMode=RCIM_Cubic, tangentMode=RCTM_User, time=1, value=0}
         ]}
        GreenCurve {keys [
          {interpMode=RCIM_Cubic, leaveTangentWeight=0.2267, tangentMode=RCTM_User, time=0, value=154.5329}
          {arriveTangentWeight=35.0282, interpMode=RCIM_Cubic, leaveTangentWeight=16.4839, tangentMode=RCTM_User, time=0.68, value=0.426}
          {arriveTangentWeight=0.1067, interpMode=RCIM_Cubic, tangentMode=RCTM_User, time=1, value=0}
         ]}
        RedCurve {keys [
          {interpMode=RCIM_Cubic, leaveTangentWeight=0.2267, tangentMode=RCTM_User, time=0, value=500}
          {arriveTangentWeight=113.3336, interpMode=RCIM_Cubic, leaveTangentWeight=53.3334, tangentMode=RCTM_User, time=0.68, value=10}
          {arriveTangentWeight=0.1067, interpMode=RCIM_Cubic, tangentMode=RCTM_User, time=1, value=0}
         ]}
       }}))
    ScaleA=true
    ScaleRGB=true
   )
   &SpriteRotationRate(Rotation Rate=&RandomRangeFloat(Minimum=-3, Maximum=3))
  ]
  rendererClasses [&NiagaraSpriteRendererProperties]
  renderers [{rendererClass=&NiagaraSpriteRendererProperties}]
 }
 Smoke_Detail_Spotlight {
  emitterUpdateScript [
   &EmitterState(Inactive Response=::NewEnumerator1, Life Cycle Mode=::NewEnumerator1, Loop Delay=0.4, Loop Duration=1, UseLoopDelay=false)
   &SpawnRate(Use Spawn Probability=false) {module=SpawnRate001}
  ]
  particleSpawnScript [
   &InitializeParticle(
    Color=<rgba 0.2353 0.2353 0.2353 0.325>
    Lifetime Max=3
    Lifetime Min=1.5
    Lifetime Mode=::NewEnumerator1
    Sprite Rotation Mode=::NewEnumerator2
    Sprite Size Max=<17 15>
    Sprite Size Min=<12 12>
    Sprite Size Mode=::NewEnumerator2
    UsePositionOffset=false
    Write Material Random=true
   )
   &PartitionParticles()
   &SkeletalMeshLocation(
    Bone / Socket Index=&SelectIntFromArray(Direct Array Index=$Particles.PartitionParticles.Partition, Int Selection Array=DataInterface() {propertyValues {IntData [1,2]}})
    Bone / Socket Sampling Mode=::NewEnumerator5
    MaskSpawnGroup=false
    Mesh Sampling Type=::NewEnumerator6
    Sampled Position Offset=&SelectVectorFromArray(
     Direct Array Index=$Particles.PartitionParticles.Partition
     Vector Selection Array=DataInterface() {propertyValues {FloatData [<-5 0 0>, <20 0 0>]}}
    )
    Skeletal Mesh=$System.Skeletal Mesh
   )
  ]
  particleUpdateScript [
   &ParticleState()
   &ScaleSpriteSize(
    Scale Sprite Size Mode=::NewEnumerator0
    Uniform Scale Factor=&MakeFloatFromVector(Vector=$Engine.Owner.Scale, Channel=::NewEnumerator2)
   ) {enabled=false, module=ScaleSpriteSize001}
   &SpriteRotationRate(Rotation Rate=&RandomRangeFloat(Minimum=-15, Maximum=15))
   &ScaleSpriteSize(
    Scale Sprite Size Mode=::NewEnumerator0
    Uniform Scale Factor=&FloatFromCurve(FloatCurve=DataInterface() {propertyValues {Curve {keys [<at 0 0.8>,<at 1 1.2>]}}}, Scale Curve=1.3)
   )
   &CurlNoiseForce(Noise Strength=25, Pan Noise=true, Pan Noise Field=<0 0 1>)
   &SubUVAnimation(
    End Frame Range Override=30
    "Sprite Renderer "=DataInterface() {propertyValues {SpriteRenderer=&self:Smoke_Detail_Spotlight_4.NiagaraSpriteRendererProperties_0}}
    Start Frame Range Override=&RandomRangeInt(Minimum=0, Maximum=5)
    UseEndFrame=true
    UseStartFrame=true
   )
   &ScaleColor(
    Scale Alpha=&Multiply_Float(
     A=&RandomRangeFloat(Minimum=0.5, Maximum=1)
     B=&FloatFromCurve(FloatCurve=DataInterface() {propertyValues {Curve {keys [
          {arriveTangent=9.0725, interpMode=RCIM_Cubic, leaveTangent=9.0725, tangentMode=RCTM_User, time=0, value=0}
          {arriveTangent=-2.5249, interpMode=RCIM_Cubic, leaveTangent=-2.5249, tangentMode=RCTM_User, time=0.1408, value=0.9847}
          {arriveTangent=-0.0405, interpMode=RCIM_Cubic, leaveTangent=-0.0405, tangentMode=RCTM_User, time=1, value=0}
         ]}}})
    )
    Scale RGB=&MakeVectorFromLinearColorRGB(Color=&ColorFromCurve(ColorCurve=DataInterface() {propertyValues {
        AlphaCurve {keys [<at 0 1 RCIM_Cubic>,<at 1 0 RCIM_Cubic>]}
        BlueCurve {keys [<at 0 1 RCIM_Cubic>,<at 0.702 0.2292 RCIM_Cubic>]}
        GreenCurve {keys [<at 0 1 RCIM_Cubic>,<at 0.702 0.2292 RCIM_Cubic>]}
        RedCurve {keys [<at 0 1 RCIM_Cubic>,<at 0.702 0.2292 RCIM_Cubic>]}
       }}))
    ScaleA=true
    ScaleRGB=true
   )
   &GravityForce(Gravity=&Multiply_Vector(
     A=&RandomRangeVector(Minimum=<0 0 10>, Maximum=<0 0 50>)
     B=&MakeVector(
      X=1
      Y=1
      Z=&FloatFromCurve(FloatCurve=DataInterface() {propertyValues {Curve {keys [
           {arriveTangent=0.6752, interpMode=RCIM_Cubic, leaveTangent=0.6752, tangentMode=RCTM_User, time=0, value=1}
           {interpMode=RCIM_Cubic, tangentMode=RCTM_User, time=0.2, value=0}
          ]}}})
     )
    ))
   &Drag(Use Linear Drag=true, Use Rotational Drag=true)
   &SolveForcesAndVelocity(Clamp Velocity=false, Limit Acceleration=false)
   &CameraOffset(Camera Offset Amount=2)
  ]
  rendererClasses [&NiagaraSpriteRendererProperties]
  renderers [{rendererClass=&NiagaraSpriteRendererProperties}]
 }
}
summary {
 systemName=NS_MicStand_LavaRock_Pickaxe_Idle
 emitters [
  {emitterName=Smoke_Detail_Spotlight, rendererClasses [&NiagaraSpriteRendererProperties]}
  {emitterName=Ashes, simTarget=GPUComputeSim, rendererClasses [&NiagaraSpriteRendererProperties]}
  {emitterName=Ashes_Base, simTarget=GPUComputeSim, rendererClasses [&NiagaraSpriteRendererProperties]}
  {emitterName=Ashes001, simTarget=GPUComputeSim, rendererClasses [&NiagaraSpriteRendererProperties]}
 ]
 userVariables [
  User.IsFrontEnd {type=NiagaraBool, defaultValue=true}
  User.IsLargeBody {type=NiagaraBool}
 ]
}
